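# Material function that remaps particle sprite UVs to mimic a sampled mesh surface, paired with the Niagara module Initialize_MeshReproductionSprite.
# Dynamic parameter index 3 supplies UV position and signed width; centered sprite UVs scale by that width and offset by UV R, wired to both axes.
# The width's sign mirrors U and yields a vector that flips sampled normal X to match the mesh basis.
# Mesh UVs are assumed square.
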MaterialFunction /Niagara/MaterialFunctions/MeshInterface/Niagara_MeshReproductionSpriteUVs
graph {
 properties {
  bExposeToLibrary=true
  description="To be used with the Niagara module called \"Initialize_MeshReproductionSprite\". This system is designed sample a mesh's uvs and tangent basis information to create a mesh proxy with particles. The assumption is that the mesh's uvs are square and not scaled to fit non-uniform textures."
  libraryCategoriesText [Niagara]
 }
 nodes [
  n1 = DynamicParameter() {parameterIndex=3, paramNames [UV R, UV G, UV width and binormal direction, Unused]}
  n2 = AppendVector(A=@n1.UV width and binormal direction, B=Abs(None=@n1.UV width and binormal direction))
  n3 = Subtract(A=TextureCoordinate()) {constB=0.5}
  FunctionOutput(None=Add(A=AppendVector(A=@n1.UV R, B=@n1.UV R), B=Multiply(A=@n2, B=@n3))) {outputName="Particle Mesh  UVs"}
  n4 = Constant3Vector() {constant=<rgba 1 1 1 0>}
  FunctionOutput(None=If(A=@n1.UV width and binormal direction, B=Constant(), "A > B"=@n4, "A == B"=@n4, "A < B"=Constant3Vector() {constant=<rgba -1 1 1 0>})) {
   description="When sampling a normal map texture from the mesh's surface, multiply the sampled value by this vector. This will align the sprites tangent space to the mesh's basis."
   outputName=Modulate Mesh Normals
   sortPriority=1
  }
 ]
}
